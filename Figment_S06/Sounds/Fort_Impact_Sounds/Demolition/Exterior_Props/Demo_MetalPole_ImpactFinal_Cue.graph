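# Sound cue for the final demolition impact of an exterior metal pole prop, using close demolition attenuation and capped by impact concurrency.
# A quality-level root picks one of four branches: the top two layer a random pole death sound (three, then two variants) with a fixed small-metal death sound.
# The lower two each play a single optimized impact wave.
# Slight random pitch modulation varies every branch, sparing the small-metal layer.
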SoundCue /Figment_S06/Sounds/Fort_Impact_Sounds/Demolition/Exterior_Props/Demo_MetalPole_ImpactFinal_Cue
root=QualityLevel_0
nodes [
 {id=Mixer_0, children [Modulator_3,Modulator_4], properties {inputVolume [1,1]}}
 {id=Mixer_1, children [Modulator_2,Modulator_7], properties {inputVolume [1,1]}}
 {id=Modulator_1, children [WavePlayer_1], properties {pitchMax=1.07, pitchMin=0.92}}
 {id=Modulator_2, children [Random_12], properties {pitchMax=1.07, pitchMin=0.92}}
 {id=Modulator_3, children [Random_1], properties {pitchMax=1.07, pitchMin=0.92}}
 {id=Modulator_4, children [WavePlayer_4]}
 {id=Modulator_5, children [WavePlayer_5], properties {pitchMax=1.07, pitchMin=0.92}}
 {id=Modulator_7, children [WavePlayer_72]}
 {id=QualityLevel_0, children [Mixer_1,Mixer_0,Modulator_1,Modulator_5]}
 {id=Random_1, children [WavePlayer_3,WavePlayer_2], properties {weights [1,1]}}
 {id=Random_12, children [WavePlayer_43, WavePlayer_42, WavePlayer_16], properties {weights [1,1,1]}}
 {id=WavePlayer_1, properties {soundWaveAssetPtr=/Figment_S06/Sounds/Fort_Impact_Sounds/Demolition/Metal_Pole_ImpactFinal_opt}}
 {id=WavePlayer_16, properties {soundWaveAssetPtr=/Figment_S06/Sounds/Fort_Impact_Sounds/Demolition/Metal_Pole_Death02}}
 {id=WavePlayer_2, properties {soundWaveAssetPtr=/Figment_S06/Sounds/Fort_Impact_Sounds/Demolition/Metal_Pole_Death06}}
 {id=WavePlayer_3, properties {soundWaveAssetPtr=/Figment_S06/Sounds/Fort_Impact_Sounds/Demolition/Metal_Pole_Death03}}
 {id=WavePlayer_4, properties {soundWaveAssetPtr=/Figment_S06/Sounds/Fort_Impact_Sounds/Demolition/Metal_Small_Death03}}
 {id=WavePlayer_42, properties {soundWaveAssetPtr=&Metal_Pole_Death06}}
 {id=WavePlayer_43, properties {soundWaveAssetPtr=&Metal_Pole_Death03}}
 {id=WavePlayer_5, properties {soundWaveAssetPtr=&Metal_Pole_ImpactFinal_opt}}
 {id=WavePlayer_72, properties {soundWaveAssetPtr=&Metal_Small_Death03}}
]
properties {
 attenuationSettings=/Figment_S06/Sounds/Attenuation/Fort_Demolition_Close_Att
 maxDistance=2000
 soundClassObject=&Fort_Impacts_PSM
 volumeMultiplier=0.48
 concurrencyOverrides {resolutionRule=PreventNew}
 concurrencySet [&CONC_Object_Impacts_FourMax]
}
