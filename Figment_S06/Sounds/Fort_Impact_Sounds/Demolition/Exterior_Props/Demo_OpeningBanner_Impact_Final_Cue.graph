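# Sound cue for the final demolition impact of an opening banner exterior prop.
# A Mixer plays two single waves together, layering the banner's final impact with a book impact sample at equal volume.
# It uses close-range demolition attenuation, the Fort impacts sound class and a shared object-impact concurrency set, named for a four-voice limit.
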SoundCue /Figment_S06/Sounds/Fort_Impact_Sounds/Demolition/Exterior_Props/Demo_OpeningBanner_Impact_Final_Cue
root=Mixer_1
nodes [
 {id=Mixer_1, children [WavePlayer_3,WavePlayer_4], properties {inputVolume [1,1]}}
 {id=WavePlayer_3, properties {soundWaveAssetPtr=&/Demo_Banner_Impact_Final_01}}
 {id=WavePlayer_4, properties {soundWaveAssetPtr=/Figment_S06/Sounds/Fort_Impact_Sounds/Demolition/Book_Impact03}}
]
properties {
 attenuationSettings=/Figment_S06/Sounds/Attenuation/Fort_Demolition_Close_Att
 maxDistance=2000
 soundClassObject=&Fort_Impacts_PSM
 volumeMultiplier=1
 concurrencyOverrides {resolutionRule=PreventNew}
 concurrencySet [&CONC_Object_Impacts_FourMax]
}
